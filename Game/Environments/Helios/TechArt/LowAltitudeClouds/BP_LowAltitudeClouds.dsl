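; Low-altitude cloud actor for the Helios environment, built from opaque/masked and translucent cloud meshes plus a box-bounded post-process effect.
; SetParameters runs from the construction script and BeginPlay, choosing meshes by complexity, assigning materials and toggling the translucent backplate.
; It feeds shape, texture and shading settings to materials through default custom primitive data.
; The post-process mesh also receives box extent, actor axes, color and fade this way.
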(event EventBeginPlay
  (CallFunction|SetParameters))

(fn SetParameters
  (bind _postprocessarea (Variables|Default|GetPostProcessArea))
  (bind _cloudmesh (Variables|Default|GetCloudMesh))
  (bind _cloudmeshtranslucent (Variables|Default|GetCloudMeshTranslucent))
  (bind _postprocessmesh (Variables|Default|GetPostProcessMesh))
  (Utilities|FlowControl|Sequence
    (:then_0
      (Components|StaticMesh|SetStaticMesh _cloudmesh (Utilities|Select :Index (Variables|Parameters|GetMeshComplexity) :NewEnumerator0 "/Game/Environments/Helios/TechArt/LowAltitudeClouds/SM_LowAltitudeClouds_Low.SM_LowAltitudeClouds_Low" :NewEnumerator1 "/Game/Environments/Helios/TechArt/LowAltitudeClouds/SM_LowAltitudeClouds_Medium.SM_LowAltitudeClouds_Medium" :NewEnumerator2 "/Game/Environments/Helios/TechArt/LowAltitudeClouds/SM_LowAltitudeClouds_High.SM_LowAltitudeClouds_High"))
      (Components|StaticMesh|SetStaticMesh _cloudmeshtranslucent (Utilities|Select :Index (Variables|Parameters|GetMeshComplexity) :NewEnumerator0 "/Game/Environments/Helios/TechArt/LowAltitudeClouds/SM_LowAltitudeClouds_LowTranslucent.SM_LowAltitudeClouds_LowTranslucent" :NewEnumerator1 "/Game/Environments/Helios/TechArt/LowAltitudeClouds/SM_LowAltitudeClouds_MediumTranslucent.SM_LowAltitudeClouds_MediumTranslucent" :NewEnumerator2 "/Game/Environments/Helios/TechArt/LowAltitudeClouds/SM_LowAltitudeClouds_HighTranslucent.SM_LowAltitudeClouds_HighTranslucent"))
      (Rendering|Material|SetMaterial _cloudmesh 0 (Variables|CloudMaterials|GetOpaqueMaterial))
      (Rendering|Material|SetMaterial _cloudmesh 1 (Variables|CloudMaterials|GetMaskedMaterial))
      (Rendering|Material|SetMaterial _cloudmeshtranslucent 0 (Variables|CloudMaterials|GetTranslucentMaterial))
      (Rendering|Material|SetDefaultCustomPrimitiveDataFloat _cloudmesh 0 (Variables|Parameters|GetMeshSeed))
      (Rendering|Material|SetDefaultCustomPrimitiveDataFloat _cloudmesh 1 (Variables|Parameters|GetMaxSize))
      (Rendering|Material|SetDefaultCustomPrimitiveDataFloat _cloudmesh 2 (Variables|Parameters|GetMinSize))
      (Rendering|Material|SetDefaultCustomPrimitiveDataFloat _cloudmesh 3 (Variables|Parameters|GetFade))
      (Rendering|Material|SetDefaultCustomPrimitiveDataFloat _cloudmesh 4 (Variables|Parameters|GetNormalIntensity))
      (Rendering|Material|SetDefaultCustomPrimitiveDataFloat _cloudmesh 6 (Variables|Parameters|GetTextureScale))
      (Rendering|Material|SetDefaultCustomPrimitiveDataFloat _cloudmesh 7 (Variables|Parameters|GetAlphaScale))
      (Rendering|SetVisibility _cloudmeshtranslucent (Variables|Parameters|GetTranslucentBackplate) false)
      (Rendering|Material|SetDefaultCustomPrimitiveDataFloat _cloudmesh 8 (Variables|Parameters|GetTextureSeed))
      (Rendering|Material|SetDefaultCustomPrimitiveDataVector3 _cloudmeshtranslucent 20 (* (Variables|Parameters|GetShadingScale) "0.700000,0.700000,0.700000"))
      (Rendering|Material|SetDefaultCustomPrimitiveDataVector3 _cloudmeshtranslucent 24 (* (Variables|Parameters|GetShadingOffset) "0.010000,0.010000,0.010000"))
      (Rendering|Material|SetDefaultCustomPrimitiveDataVector3 _cloudmesh 28 (* (Variables|Parameters|GetTransmitanceScale) "0.400000,0.400000,0.400000"))
      (Rendering|Material|SetDefaultCustomPrimitiveDataVector3 _cloudmeshtranslucent 32 (* (Variables|Parameters|GetTransmitanceOffset) "0.010000,0.010000,0.010000"))
      (Rendering|Material|SetDefaultCustomPrimitiveDataVector3 (Variables|Default|GetPostProcessMesh) 16 (Math|Conversions|ToVector(LinearColor) (Variables|Parameters|GetColorMultiplier))))
    (:then_1
      (Rendering|Material|SetMaterial _postprocessarea 0 (Variables|CloudMaterials|GetPostProcessMaterial))
      (Transformation|SetRelativeLocation _postprocessarea (Variables|Parameters|GetPostProcessOffset) false false)
      (Transformation|SetRelativeScale3D _postprocessarea (Variables|Parameters|GetPostProcessScale))
      (Rendering|Material|SetDefaultCustomPrimitiveDataVector3 _postprocessmesh 0 (Components|Box|GetScaledBoxExtent (Variables|Default|GetPostProcessArea)))
      (Rendering|Material|SetDefaultCustomPrimitiveDataVector3 _postprocessmesh 4 (Math|Vector|GetForwardVector (Transformation|GetActorRotation)))
      (Rendering|Material|SetDefaultCustomPrimitiveDataVector3 _postprocessmesh 8 (Math|Vector|GetRightVector (Transformation|GetActorRotation)))
      (Rendering|Material|SetDefaultCustomPrimitiveDataVector3 _postprocessmesh 12 (Math|Vector|GetUpVector (Transformation|GetActorRotation)))
      (Rendering|Material|SetDefaultCustomPrimitiveDataFloat _postprocessmesh 20 (Variables|Parameters|GetPostProcessFade)))))

(fn ConstructionScript
  (CallFunction|SetParameters))
