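# Material function that supplies time-of-day colors for the Helios low altitude clouds.
# The TimeOfDay_0-1 collection parameter drives U into the TOD Curves texture, sampled at V 0 and V 1 for two color gradients.
# Both samples pass through vertex interpolators, so lookups run per vertex, then the DotMask input blends between them for the output color.
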MaterialFunction /Game/Environments/Helios/TechArt/LowAltitudeClouds/MF_LowAltitudeClouds_Colors
graph {nodes [
  n1 = CollectionParameter() {collection=&FortniteMaterialParameters, parameterName=TimeOfDay_0-1}
  n2 = TextureObjectParameter() {
   parameterName=TOD Curves
   samplerType=SAMPLERTYPE_LinearColor
   texture=&/CA_LowAltitudeClouds
  }
  n3 = TextureSample(UVs=AppendVector(A=@n1, B=Constant()), Tex=@n2) {
   samplerSource=SSM_Clamp_WorldGroupSettings
   samplerType=SAMPLERTYPE_LinearColor
   texture=&CA_LowAltitudeClouds
  }
  n4 = TextureSample(UVs=AppendVector(A=@n1, B=Constant() {r=1}), Tex=@n2) {
   samplerSource=SSM_Clamp_WorldGroupSettings
   samplerType=SAMPLERTYPE_LinearColor
   texture=&CA_LowAltitudeClouds
  }
  FunctionOutput(None=LinearInterpolate(
    A=VertexInterpolator(VS=@n3.RGB).PS
    B=VertexInterpolator(VS=@n4.RGB).PS
    Alpha=FunctionInput() {bUsePreviewValueAsDefault=true, inputName=DotMask, inputType=FunctionInput_Scalar}
   ))
 ]}
