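# Unlit masked, dithered material that renders Helios low-altitude clouds as camera-facing cards on instanced meshes.
# World position offset scatters cards using seeded local Z, orients and sizes them, and adjusts them near the camera.
# Spherical normals plus distance-faded detail from a sliced 2D volume texture drive sky-light shading and backlit transmittance.
# Custom primitive data tunes each cloud, IsTranslucent switches alpha panning and offset, and HadesActive tints the color.
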Material /Game/Environments/Helios/TechArt/LowAltitudeClouds/M_LowAltitudeClouds
graph {
 properties {
  bHasPixelAnimation=true
  bTangentSpaceNormal=false
  bUsedWithInstancedStaticMeshes=true
  blendMode=BLEND_Masked
  pixelDepthOffsetMode=PDOM_Legacy
  shadingModel=MSM_Unlit
  translucencyPass=MTP_BeforeDOF
  usedShadingModels=Unlit
 }
 nodes [
  n1 = Multiply(A=DotProduct(A=NamedRerouteUsage(), B=SkyAtmosphereLightDirection())) {constB=0.25}
  n2 = Saturate(None=Add(A=@n1) {constB=0.5})
  n3 = OneMinus(None=Multiply(A=@n2, B=@n2))
  n4 = Multiply(A=@n3, B=@n3)
  n5 = Multiply(A=@n4, B=@n4)
  Multiply(A=@n5, B=@n5)
  n6 = VertexColor()
  n7 = Multiply(A=Add(A=NamedRerouteUsage(), B=@n6.R), B=@n6.R)
  NamedRerouteDeclaration(None=Multiply(A=Power(Base=@n7) {constExponent=1.5}) {constB=2}) {name=Opacity, nodeColor=<rgba 0 0.4141 1 1>}
  n8 = NamedRerouteUsage()
  n9 = Multiply(A=ScalarParameter() {bUseCustomPrimitiveData=true, defaultValue=800, parameterName=Alpha Scale, primitiveDataIndex=7}, B=NamedRerouteUsage())
  n10 = Saturate(None=Divide(A=NamedRerouteUsage()) {constB=50000})
  n11 = Min(A=Multiply(A=@n10, B=@n10)) {constB=0.4}
  n12 = Multiply(A=Multiply(A=Divide(A=@n8, B=@n9) {constB=400}, B=OneMinus(None=@n11))) {constB=2}
  n13 = Time()
  n14 = Constant2Vector() {g=-0.0365, r=0.025}
  n15 = VertexInterpolator(VS=Add(A=@n12, B=Multiply(A=@n13, B=@n14) {constB=0.02}))
  n16 = VertexInterpolator(VS=@n12)
  NamedRerouteDeclaration(None=StaticSwitchParameter(
    True=TextureSample(UVs=@n15.PS) {samplerType=SAMPLERTYPE_LinearGrayscale, texture=&/T_LowAltitudeClouds_A}.R
    False=TextureSample(UVs=@n16.PS) {samplerType=SAMPLERTYPE_LinearGrayscale, texture=&T_LowAltitudeClouds_A}.R
   ) {parameterName=IsTranslucent}) {name=Alpha, nodeColor=<rgba 0 1 0.9531 1>}
  n17 = ActorPositionWS() {originType=CameraRelative}
  n18 = LocalPosition()
  n19 = ScalarParameter() {bUseCustomPrimitiveData=true, defaultValue=1, parameterName=Seed}
  n20 = Frac(None=Multiply(A=Multiply(A=@n18.Z) {constB=0.6599}, B=@n19))
  n21 = Frac(None=Multiply(A=Multiply(A=@n18.Z) {constB=0.1369}, B=@n19))
  n22 = Frac(None=Multiply(A=Multiply(A=@n18.Z) {constB=0.4688}, B=@n19))
  n23 = Multiply(A=Add(A=MaterialFunctionCall(R=@n20, G=@n21, B=@n22) {materialFunction=&AppendMany}.RGB) {constB=-0.5}) {constB=200}
  n24 = NamedRerouteUsage()
  n25 = MaterialFunctionCall(Float3=Abs(None=Divide(A=@n24) {constB=100})) {materialFunction=&BreakOutFloat3Components}
  n26 = Max(A=Max(A=@n25.R, B=@n25.R), B=@n25.B)
  n27 = AppendVector(
   A=ComponentMask(None=NamedRerouteDeclaration(None=@n23) {name=Random Positions, nodeColor=<rgba 0.3125 1 0 1>}) {g=true, r=true}
   B=Add(A=Multiply(A=@n26) {constB=-100}) {constB=30}
  )
  n28 = Transform(None=LinearInterpolate(A=@n27, B=Multiply(A=@n27) {constB=0.5}, Alpha=NamedRerouteUsage()) {constB=0}) {transformSourceType=TRANSFORMSOURCE_Instance}
  n29 = Normalize(VectorInput=NamedRerouteDeclaration(None=Add(A=@n17, B=@n28)) {name=Random Local Positions, nodeColor=<rgba 0.3828 0 1 1>})
  n30 = Multiply(A=Multiply(A=NamedRerouteUsage(), B=NamedRerouteUsage())) {constB=-100}
  n31 = Max(A=VertexColor().G) {constB=0.18}
  n32 = Add(
   A=Multiply(A=NamedRerouteDeclaration(None=@n29) {name=Card vector, nodeColor=<rgba 1 0 0.1172 1>}, B=Multiply(A=@n30, B=@n31)) {constB=-5000}
   B=@n28
  )
  n33 = Multiply(A=Subtract(A=WorldPosition() {worldPositionShaderOffset=WPT_CameraRelative}.XYZ, B=ActorPositionWS() {originType=CameraRelative})) {constB=-1}
  n34 = Multiply(A=NamedRerouteUsage()) {constB=170}
  n35 = Multiply(A=NamedRerouteUsage(), B=Multiply(A=NamedRerouteUsage(), B=@n34)) {constB=750}
  n36 = Distance(A=NamedRerouteUsage(), B=Constant())
  n37 = Saturate(None=Divide(A=NamedRerouteDeclaration(None=@n36) {name=Cards Distance, nodeColor=<rgba 1 0 0.0703 1>}, B=@n34))
  n38 = OneMinus(None=@n37)
  n39 = NamedRerouteUsage()
  n40 = Normalize(VectorInput=CrossProduct(A=Constant3Vector() {constant=<rgba 0 0 1 0>}, B=@n39))
  n41 = LocalPosition()
  n42 = Multiply(A=@n40, B=ComponentMask(None=@n41.XYZ) {r=true})
  n43 = Multiply(A=CrossProduct(A=@n40, B=@n39), B=ComponentMask(None=@n41.XYZ) {g=true})
  n44 = MaterialFunctionCall() {materialFunction=&ObjectScale}
  n45 = Min(A=Min(A=@n44.Scale X, B=@n44.Scale X), B=@n44.Scale Z)
  n46 = NamedRerouteDeclaration(None=LinearInterpolate(
    A=ScalarParameter() {bUseCustomPrimitiveData=true, defaultValue=1, parameterName=Max Size, primitiveDataIndex=1}
    B=ScalarParameter() {bUseCustomPrimitiveData=true, defaultValue=0.1945, parameterName=Min Size, primitiveDataIndex=2}
    Alpha=@n26
   ) {constA=1, constB=0.1}) {name=Cards Sizes, nodeColor=<rgba 1 0.4922 0 1>}
  n47 = Add(A=Divide(A=NamedRerouteUsage()) {constB=100000})
  n48 = Multiply(
   A=OneMinus(None=VertexColor().R)
   B=ScalarParameter() {bUseCustomPrimitiveData=true, defaultValue=0.35, parameterName=Fade, primitiveDataIndex=3}
  ) {constB=0.3}
  n49 = Multiply(
   A=Multiply(A=Add(A=@n42, B=@n43), B=NamedRerouteDeclaration(None=@n45) {name=Min Scale, nodeColor=<rgba 0 0.4375 1 1>})
   B=Add(A=Divide(A=@n46, B=@n47), B=Multiply(A=@n48, B=@n47))
  )
  n50 = Add(
   A=Add(A=Add(A=@n32, B=@n33), B=Multiply(A=@n35, B=@n38) {constB=500}) {constB=0}
   B=Multiply(A=@n49, B=LinearInterpolate(Alpha=@n37) {constA=0.7})
  ) {constB=0}
  n51 = Add(A=WorldPosition() {worldPositionShaderOffset=WPT_CameraRelative}.XYZ, B=@n50)
  NamedRerouteDeclaration(None=Subtract(A=@n51, B=ActorPositionWS() {originType=CameraRelative})) {name=Origin Position, nodeColor=<rgba 1 0.1641 0 1>}
  n52 = Multiply(A=Normalize(VectorInput=@n51)) {constB=10}
  NamedRerouteDeclaration(None=StaticSwitchParameter(True=Add(A=@n50, B=@n52), False=@n50) {parameterName=IsTranslucent}) {name=WPO, nodeColor=<rgba 0 1 0.1094 1>}
  NamedRerouteDeclaration(None=MaterialFunctionCall(Alpha Threshold=Multiply(A=@n7, B=@n6.R)) {materialFunction=&DitherTemporalAA}.Result) {name=Opacity Mask, nodeColor=<rgba 0 1 0.3438 1>}
  n53 = OneMinus(None=Saturate(None=@n1))
  n54 = NamedRerouteUsage()
  n55 = Add(A=OneMinus(None=Multiply(A=@n53, B=@n53)), B=OneMinus(None=Multiply(A=@n5, B=@n54)))
  n56 = CameraVectorWS()
  n57 = SkyAtmosphereLightDirection()
  n58 = Saturate(None=Multiply(A=DotProduct(A=@n56, B=@n57)) {constB=-1})
  n59 = Multiply(A=@n58, B=@n58)
  n60 = Multiply(A=@n59, B=@n59)
  n61 = Multiply(A=@n60, B=@n60)
  n62 = OneMinus(None=Saturate(None=NamedRerouteUsage()))
  n63 = OneMinus(None=Multiply(A=@n62, B=@n62))
  n64 = NamedRerouteUsage()
  n65 = Multiply(A=@n64, B=@n64)
  n66 = Multiply(A=@n65, B=@n65) {constB=0.5}
  n67 = Multiply(A=Multiply(A=Multiply(A=@n61, B=@n63) {constA=4, constB=4}) {constA=4, constB=10}, B=OneMinus(None=Multiply(A=@n66) {constB=0.6}))
  n68 = VectorParameter() {bUseCustomPrimitiveData=true, parameterName=Color Multiplier, primitiveDataIndex=16, defaultValue=<rgba 0 0 0 1>}
  n69 = Multiply(
   A=MaterialFunctionCall!(DotMask=Add(A=@n55, B=@n67) {constA=1}) {materialFunction=&/MF_LowAltitudeClouds_Colors}.Result
   B=@n68.RGB
  )
  n70 = Multiply(A=@n69, B=VectorParameter() {parameterName=Hades Multiplier, defaultValue=<rgba 0.6156 0.776 0.6448 1>}.RGB)
  n71 = CollectionParameter() {collection=&FortniteMaterialParameters, parameterName=HadesActive}
  n72 = Constant() {r=1}
  n73 = Constant() {r=1.1}
  NamedRerouteDeclaration(None=Multiply(A=LinearInterpolate(A=@n69, B=@n70, Alpha=@n71), B=QualitySwitch(Default=@n72, Low=@n73, High=@n72, Medium=@n73, Epic=@n72)) {constB=1.2}) {name=Color, nodeColor=<rgba 0 1 0.6953 1>}
  n74 = Divide(
   A=Transform(None=NamedRerouteUsage()) {transformSourceType=TRANSFORMSOURCE_World, transformType=TRANSFORM_Instance}
   B=Multiply(A=MaterialFunctionCall() {materialFunction=&ObjectScale}.Scale XYZ) {constB=100}
  )
  n75 = VectorParameter() {bUseCustomPrimitiveData=true, parameterName=Transmitance Scale, primitiveDataIndex=28, defaultValue=<rgba 0 0 0 1>}
  n76 = Divide(A=@n74, B=@n75.RGB) {constB=160}
  n77 = VectorParameter() {bUseCustomPrimitiveData=true, parameterName=Transmitance Offset, primitiveDataIndex=32, defaultValue=<rgba 0 0 0 1>}
  n78 = VertexInterpolator(VS=Transform(None=Subtract(A=@n76, B=@n77.RGB)) {transformSourceType=TRANSFORMSOURCE_Instance, transformType=TRANSFORM_View})
  n79 = Multiply(A=@n78.PS, B=@n78.PS)
  NamedRerouteDeclaration(None=Add(A=ComponentMask(None=@n79) {r=true}, B=ComponentMask(None=@n79) {g=true})) {name=Transmitance Mask, nodeColor=<rgba 0.0547 1 0 1>}
  n80 = VectorParameter() {bUseCustomPrimitiveData=true, parameterName=Shading Scale, primitiveDataIndex=20, defaultValue=<rgba 0 0 0 1>}
  n81 = VectorParameter() {bUseCustomPrimitiveData=true, parameterName=Shading Offset, primitiveDataIndex=24, defaultValue=<rgba 0 0 0 1>}
  n82 = Subtract(A=Divide(A=@n74, B=@n80.RGB) {constB=160}, B=@n81.RGB)
  NamedRerouteDeclaration(None=VertexInterpolator(VS=Transform(None=@n82) {transformSourceType=TRANSFORMSOURCE_Instance}).PS) {name=Spherical Normals, nodeColor=<rgba 1 0.8438 0 1>}
  n83 = ScalarParameter() {bUseCustomPrimitiveData=true, defaultValue=800, parameterName=Texture Scale, primitiveDataIndex=6}
  n84 = NamedRerouteUsage()
  n85 = ScalarParameter() {bUseCustomPrimitiveData=true, parameterName=RandomTile, primitiveDataIndex=8}
  n86 = Constant3Vector() {constant=<rgba 2.0028 1.4174 5.8472 0>}
  n87 = Add(A=Divide(A=NamedRerouteUsage(), B=Multiply(A=@n83, B=@n84)), B=Frac(None=Multiply(A=@n85, B=@n86)))
  n88 = CameraVectorWS()
  n89 = MaterialFunctionCall!(
   UVs=Add(A=VertexInterpolator(VS=@n87).PS, B=Multiply(A=@n88) {constB=-0.1})
   Slices Texture=TextureObject() {samplerType=SAMPLERTYPE_LinearColor, texture=&/T_LowAltitudeClouds_N}
   Atlas Size=Constant() {r=8}
   Atlas Padding=ScalarParameter() {defaultValue=0.9375, parameterName=AtlasPadding, primitiveDataIndex=8}
  ) {materialFunction=&/MF_3DVolumeFrom2DSlices}
  n90 = Add(A=@n89.RGB) {constB=-0.5}
  n91 = Distance(A=WorldPosition() {worldPositionShaderOffset=WPT_CameraRelative}.XYZ, B=Constant())
  n92 = Saturate(None=Divide(A=NamedRerouteDeclaration(None=@n91) {name=Distance Fade, nodeColor=<rgba 0 0.9297 1 1>}) {constB=100000})
  n93 = Min(A=Multiply(A=@n92, B=@n92)) {constB=0.25}
  n94 = VertexInterpolator(VS=Multiply(
    A=OneMinus(None=@n93)
    B=ScalarParameter() {bUseCustomPrimitiveData=true, defaultValue=5, parameterName=Normal Intensity, primitiveDataIndex=4}
   ))
  NamedRerouteDeclaration(None=Add(A=Multiply(A=@n90, B=@n94.PS) {constB=6}, B=NamedRerouteUsage())) {name=Normals, nodeColor=<rgba 0.2656 0 1 1>}
  NamedRerouteDeclaration(None=@n89.A) {name=Masks, nodeColor=<rgba 0.9219 0 1 1>}
  NamedRerouteDeclaration(None=ComponentMask(None=Transform(None=@n49) {transformSourceType=TRANSFORMSOURCE_World, transformType=TRANSFORM_View}) {g=true, r=true}) {name=Alpha Coords, nodeColor=<rgba 0.6172 1 0 1>}
  NamedRerouteUsage()
  NamedRerouteUsage()
  NamedRerouteUsage()
  NamedRerouteUsage()
  Reroute()
 ]
}
