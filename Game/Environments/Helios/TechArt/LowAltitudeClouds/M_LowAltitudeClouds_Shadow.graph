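# Masked, unlit, two-sided material that, by its name, serves as the shadow proxy for low-altitude clouds.
# Pivot-relative position in view space, panned slowly over time and computed at vertices, samples a grayscale cloud texture.
# That sample and a squared normal-to-atmosphere-light alignment drive DitherTemporalAA, fading areas edge-on to the light.
# A normal-based offset scales with the smaller of object X and Z scale; Y is never read.
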Material /Game/Environments/Helios/TechArt/LowAltitudeClouds/M_LowAltitudeClouds_Shadow
graph {
 properties {
  bHasPixelAnimation=true
  bTangentSpaceNormal=false
  blendMode=BLEND_Masked
  maxWorldPositionOffsetDisplacement=4000
  pixelDepthOffsetMode=PDOM_Legacy
  shadingModel=MSM_Unlit
  translucencyPass=MTP_BeforeDOF
  twoSided=true
  usedShadingModels=Unlit
 }
 nodes [
  n1 = Subtract(
   A=WorldPosition().XYZ
   B=MaterialFunctionCall() {materialFunction=&ObjectPivotPoint}.Mesh Particle Pivot Location
  )
  n2 = ComponentMask(None=Transform(None=Divide(A=@n1) {constB=1000}) {transformSourceType=TRANSFORMSOURCE_World, transformType=TRANSFORM_View}) {g=true, r=true}
  n3 = Multiply(A=Time(), B=Constant2Vector() {g=0.066, r=0.073}) {constB=0.02}
  n4 = TextureSample(UVs=VertexInterpolator(VS=Add(A=@n2, B=@n3)).PS) {samplerType=SAMPLERTYPE_LinearGrayscale, texture=&/T_LowAltitudeClouds_A}
  n5 = Abs(None=DotProduct(A=VertexNormalWS(), B=SkyAtmosphereLightDirection()))
  n6 = Multiply(A=Multiply(A=@n5, B=@n5) {constB=2}) {constB=2}
  MaterialFunctionCall(Alpha Threshold=Multiply(A=Add(A=@n4.R, B=@n6), B=@n6)) {materialFunction=&DitherTemporalAA}
  n7 = MaterialFunctionCall() {materialFunction=&ObjectScale}
  Multiply(A=Multiply(A=VertexNormalWS()) {constB=20}, B=Min(A=Min(A=@n7.Scale X, B=@n7.Scale X), B=@n7.Scale Z))
 ]
}
