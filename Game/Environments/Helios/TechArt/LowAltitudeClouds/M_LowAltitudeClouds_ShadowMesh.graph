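# Unlit material for the shadow mesh of the low-altitude clouds in the Helios environment, as its name and folder indicate.
# The graph has no expressions, so it computes no color, opacity or offset and relies on default inputs.
# Apart from the Unlit shading model, the only listed setting is legacy pixel depth offset mode.
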Material /Game/Environments/Helios/TechArt/LowAltitudeClouds/M_LowAltitudeClouds_ShadowMesh
graph {expressions {}, properties {pixelDepthOffsetMode=PDOM_Legacy, shadingModel=MSM_Unlit, usedShadingModels=Unlit}}
